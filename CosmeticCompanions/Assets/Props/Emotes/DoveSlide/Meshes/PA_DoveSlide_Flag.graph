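# Physics asset for the flag prop in the DoveSlide cosmetic emote, giving its pole and flag bones collision bodies.
# Every body is a single capsule: tiny ones on the three dyn_flag bones, a longer one on the pole, and a very large, near-spherical one on root.
# Constraints form one chain from root through flag_pole to each flag bone in turn, all sharing identical swing and twist limit angles.
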PhysicsAsset /CosmeticCompanions/Assets/Props/Emotes/DoveSlide/Meshes/PA_DoveSlide_Flag
bodies {
 dyn_flag_01 [{shapeType=Capsule, length=0.505, radius=0.505, center=<15.1763 0.0327 -0.4046>, rotation=<rot 0 -90.0564 -89.9095>}]
 dyn_flag_02 [{shapeType=Capsule, length=0.505, radius=0.505, center=<4.6705 0.0327 -0.4167>, rotation=<rot 0.0028 -90 -0.029>}]
 dyn_flag_03 [{shapeType=Capsule, length=0.505, radius=0.505, center=<3.233 0.0327 -0.4156>, rotation=<rot 0.0023 -90 -0.0194>}]
 flag_pole [{shapeType=Capsule, length=20.5814, radius=3.1664, center=<1.9711 -0.0072 11.736>, rotation=<rot 0.0207 -90 0.7328>}]
 root [{shapeType=Capsule, length=0.505, radius=110.2365, rotation=<rot 0 -90 0>}]
}
constraints [
 {bone1Name=flag_pole, bone2Name=root, swing1LimitDegrees=45, swing2LimitDegrees=45, twistLimitDegrees=45}
 {bone1Name=dyn_flag_01, bone2Name=flag_pole, swing1LimitDegrees=45, swing2LimitDegrees=45, twistLimitDegrees=45}
 {bone1Name=dyn_flag_02, bone2Name=dyn_flag_01, swing1LimitDegrees=45, swing2LimitDegrees=45, twistLimitDegrees=45}
 {bone1Name=dyn_flag_03, bone2Name=dyn_flag_02, swing1LimitDegrees=45, swing2LimitDegrees=45, twistLimitDegrees=45}
]
